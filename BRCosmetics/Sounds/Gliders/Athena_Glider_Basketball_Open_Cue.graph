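# Sound cue for the Basketball glider's open sound in Athena, layering a deploy effect with a music sting.
# A mixer plays the open SFX at full volume alongside the Basketball emote music wave, which is lowered in volume and starts after a fixed 1.5-second delay.
# Shared glider open/close attenuation and concurrency, a cosmetic glider-open property sheet and the glider land sound class govern playback.
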SoundCue /BRCosmetics/Sounds/Gliders/Athena_Glider_Basketball_Open_Cue
root=Mixer_0
nodes [
 {id=Delay_0, children [WavePlayer_0], properties {delayMax=1.5, delayMin=1.5}}
 {id=Mixer_0, children [Delay_0,WavePlayer_1], properties {inputVolume [0.3,1]}}
 {id=WavePlayer_0, properties {soundWaveAssetPtr=&/Basketball_Emote_Music_Glider_Open_01}}
 {id=WavePlayer_1, properties {soundWaveAssetPtr=&/SFX_Glider_Basketball_Open}}
]
properties {
 attenuationSettings=/Game/Sounds/Attenuation/Fort_Player_Glider_OpenClose_Athena_Att
 audioPropertiesSheet=/BRCosmetics/Sounds/Settings/PropertySheets/Glider/APS_Cosmetic_Glider_Open
 maxDistance=10000
 soundClassObject=&Fort_Gliders_Land
 concurrencySet [&Athena_GliderOpenClose_Conc]
}
