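# Sound cue for the Basketball glider cosmetic's thrust start, playing a basketball emote music loop.
# A single looping wave player is the root, with no other nodes in between.
# The cue plays at reduced volume in the glider music sound class, using player glider thrust attenuation and a glider thrust concurrency set.
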SoundCue /BRCosmetics/Sounds/Gliders/Athena_Glider_Basketball_ThrustStart_Cue
root=WavePlayer_0
nodes [{id=WavePlayer_0, properties {bLooping=true, soundWaveAssetPtr=&/Basketball_Emote_Music_Glider_01_Loop}}]
properties {
 attenuationSettings=&Fort_Player_Glider_Thrust_Athena_Att
 maxDistance=12500
 soundClassObject=/Game/Sounds/FortSoundClassesAndModes/Fort_Gliders_Music
 volumeMultiplier=0.25
 concurrencySet [&Athena_Glider_Thrust_Conc]
}
