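# Sound cue for the idle loop of the Carbide glider in Athena, spatialized with player glider thrust attenuation.
# A Mixer layers two looping waves, a glider idle loop and a jet loop, at different volumes.
# An Enveloper above the mix ramps volume from partial to full over the first second.
# A concurrency override allows one instance per owner and prevents new ones from starting while it plays.
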SoundCue /BRCosmetics/Sounds/Gliders/Athena_Glider_Carbide_IdleLoop_Cue
root=Enveloper_0
nodes [
 {
  id=Enveloper_0
  children [Mixer_0]
  properties {pitchCurve {editorCurveData {keys [{time=0}]}}, volumeCurve {editorCurveData {keys [<at -0.0005 0.368>,<at 1 1>]}}}
 }
 {id=Mixer_0, children [WavePlayer_1,WavePlayer_0], properties {inputVolume [0.5,0.8]}}
 {id=WavePlayer_0, properties {bLooping=true, soundWaveAssetPtr=&/Carbide_Jet_Loop_01}}
 {id=WavePlayer_1, properties {bLooping=true, soundWaveAssetPtr=/Game/Sounds/Fort_Gadgets/Athena/Glider_CarbideBlue_IdleLoop}}
]
properties {
 attenuationSettings=&Fort_Player_Glider_Thrust_Athena_Att
 audioPropertiesSheet=&APS_Cosmetic_Glider_Loop
 bOverrideConcurrency=true
 maxDistance=12500
 soundClassObject=&Fort_Gliders_Land
 volumeMultiplier=0.5
 concurrencyOverrides {bLimitToOwner=true, maxCount=1, resolutionRule=PreventNew}
}
