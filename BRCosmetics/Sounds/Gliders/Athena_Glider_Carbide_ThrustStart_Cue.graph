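# Sound cue for the Carbide glider's thrust-start sound in Athena, using CarbideBlue wave assets.
# A Random node picks between two thrust-start wave variations with equal weight.
# Playback uses the player glider thrust attenuation, a glider sound class and a shared glider thrust concurrency set, at reduced volume.
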SoundCue /BRCosmetics/Sounds/Gliders/Athena_Glider_Carbide_ThrustStart_Cue
root=Random_0
nodes [
 {id=Random_0, children [WavePlayer_1,WavePlayer_3], properties {weights [1,1]}}
 {id=WavePlayer_1, properties {soundWaveAssetPtr=&/Glider_CarbideBlue_ThrustStart_01}}
 {id=WavePlayer_3, properties {soundWaveAssetPtr=&/Glider_CarbideBlue_ThrustStart_02}}
]
properties {
 attenuationSettings=&Fort_Player_Glider_Thrust_Athena_Att
 maxDistance=12500
 soundClassObject=&Fort_Gliders_Land
 volumeMultiplier=0.7
 concurrencySet [&Athena_Glider_Thrust_Conc]
}
