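# Sound cue for the Dark Viking glider's open sound in Athena.
# A Mixer plays the open sound immediately alongside a randomly chosen, equally weighted dragon wing flap delayed by a fixed one second.
# Glider open/close attenuation and concurrency settings, a glider-open audio property sheet and a reduced volume multiplier shape playback.
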SoundCue /BRCosmetics/Sounds/Gliders/Athena_Glider_Dark_Viking_Open_Cue
root=Mixer_0
nodes [
 {id=Delay_1, children [Random_0], properties {delayMax=1, delayMin=1}}
 {id=Mixer_0, children [WavePlayer_1,Delay_1], properties {inputVolume [1,1]}}
 {id=Random_0, children [WavePlayer_0,WavePlayer_2], properties {weights [1,1]}}
 {id=WavePlayer_0, properties {soundWaveAssetPtr=/Game/Sounds/Fort_Gadgets/Athena/DarkViking_Dragon_WingFlap_01}}
 {id=WavePlayer_1, properties {soundWaveAssetPtr=&/DarkViking_Glider_Open_01}}
 {id=WavePlayer_2, properties {soundWaveAssetPtr=/Game/Sounds/Fort_Gadgets/Athena/DarkViking_Dragon_WingFlap_02}}
]
properties {
 attenuationSettings=/Game/Sounds/Attenuation/Fort_Player_Glider_OpenClose_Athena_Att
 audioPropertiesSheet=/BRCosmetics/Sounds/Settings/PropertySheets/Glider/APS_Cosmetic_Glider_Open
 maxDistance=10000
 soundClassObject=&Fort_Gliders_Land
 volumeMultiplier=0.55
 concurrencySet [&Athena_GliderOpenClose_Conc]
}
