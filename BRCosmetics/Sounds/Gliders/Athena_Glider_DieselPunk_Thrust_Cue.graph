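# Sound cue for the thrust sound of the DieselPunk cosmetic glider in Battle Royale.
# A Random node picks one of two equally weighted thrust waves and feeds a Modulator at its default settings.
# Playback uses the glider thrust attenuation and concurrency, a glider loop audio properties sheet and the Gliders_Land sound class, at reduced volume.
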SoundCue /BRCosmetics/Sounds/Gliders/Athena_Glider_DieselPunk_Thrust_Cue
root=Modulator_0
nodes [
 {id=Modulator_0, children [Random_0]}
 {id=Random_0, children [WavePlayer_0,WavePlayer_2], properties {weights [1,1]}}
 {id=WavePlayer_0, properties {soundWaveAssetPtr=&/SFX_Glider_DieselPunk_Thrust_03}}
 {id=WavePlayer_2, properties {soundWaveAssetPtr=&/SFX_Glider_DieselPunk_Thrust_04}}
]
properties {
 attenuationSettings=&Fort_Player_Glider_Thrust_Athena_Att
 audioPropertiesSheet=&APS_Cosmetic_Glider_Loop
 maxDistance=12500
 soundClassObject=&Fort_Gliders_Land
 volumeMultiplier=0.2
 concurrencySet [&Athena_Glider_Thrust_Conc]
}
