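# Sound cue for the opening sound of the Heist glider cosmetic in Athena.
# A single wave player feeds one open sound wave straight to the output, with no randomization or modulation.
# Playback uses the shared player glider open/close attenuation and concurrency, a cosmetic glider-open audio property sheet, and the gliders land sound class, with a slight volume boost.
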SoundCue /BRCosmetics/Sounds/Gliders/Athena_Glider_Heist_Open_Cue
root=WavePlayer_0
nodes [{id=WavePlayer_0, properties {soundWaveAssetPtr=&/Athena_Glider_Heist_Open_01}}]
properties {
 attenuationSettings=/Game/Sounds/Attenuation/Fort_Player_Glider_OpenClose_Athena_Att
 audioPropertiesSheet=/BRCosmetics/Sounds/Settings/PropertySheets/Glider/APS_Cosmetic_Glider_Open
 maxDistance=10000
 soundClassObject=&Fort_Gliders_Land
 volumeMultiplier=1.1
 concurrencySet [&Athena_GliderOpenClose_Conc]
}
