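# Sound cue for the opening of a metal glider in the Athena cosmetics set, built as a layered, staggered one-shot.
# A mixer plays the first metal open wave at full level, then brings in two quieter waves after fixed, successively longer delays, the middle one pitched down.
# It uses glider open/close attenuation and concurrency, a glider-open property sheet, and the Fort_Gliders_Land sound class.
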SoundCue /BRCosmetics/Sounds/Gliders/Athena_Glider_Metal_Open_Cue
root=Mixer_0
nodes [
 {id=Delay_0, children [Modulator_0], properties {delayMax=0.35, delayMin=0.35}}
 {id=Delay_1, children [WavePlayer_3], properties {delayMax=0.7, delayMin=0.7}}
 {id=Mixer_0, children [WavePlayer_4,Delay_0,Delay_1], properties {inputVolume [1,0.4,0.3]}}
 {id=Modulator_0, children [WavePlayer_2], properties {pitchMax=0.8, pitchMin=0.8}}
 {id=WavePlayer_2, properties {soundWaveAssetPtr=&/Athena_Glider_Metal_Open03}}
 {id=WavePlayer_3, properties {soundWaveAssetPtr=&/Athena_Glider_Metal_Open04}}
 {id=WavePlayer_4, properties {soundWaveAssetPtr=&/Athena_Glider_Metal_Open01}}
]
properties {
 attenuationSettings=/Game/Sounds/Attenuation/Fort_Player_Glider_OpenClose_Athena_Att
 audioPropertiesSheet=/BRCosmetics/Sounds/Settings/PropertySheets/Glider/APS_Cosmetic_Glider_Open
 maxDistance=10000
 soundClassObject=&Fort_Gliders_Land
 volumeMultiplier=1.1
 concurrencySet [&Athena_GliderOpenClose_Conc]
}
