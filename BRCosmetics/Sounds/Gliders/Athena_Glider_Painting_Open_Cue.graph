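# Sound cue for the opening of the Painting glider cosmetic in Athena.
# A mixer layers the open sound with a quieter extend sound that plays after a fixed short delay at a fixed, slightly lowered pitch.
# It uses the glider open/close attenuation and concurrency, a cosmetic glider-open property sheet and the glider landing sound class.
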SoundCue /BRCosmetics/Sounds/Gliders/Athena_Glider_Painting_Open_Cue
root=Mixer_0
nodes [
 {id=Delay_0, children [WavePlayer_2], properties {delayMax=0.8, delayMin=0.8}}
 {id=Mixer_0, children [WavePlayer_1,Modulator_0], properties {inputVolume [1,0.5]}}
 {id=Modulator_0, children [Delay_0], properties {pitchMax=0.9, pitchMin=0.9}}
 {id=WavePlayer_1, properties {soundWaveAssetPtr=&/Athena_Glider_Painting_Open}}
 {id=WavePlayer_2, properties {soundWaveAssetPtr=&/Athena_Glider_Painting_Extend}}
]
properties {
 attenuationSettings=/Game/Sounds/Attenuation/Fort_Player_Glider_OpenClose_Athena_Att
 audioPropertiesSheet=/BRCosmetics/Sounds/Settings/PropertySheets/Glider/APS_Cosmetic_Glider_Open
 maxDistance=10000
 soundClassObject=&Fort_Gliders_Land
 volumeMultiplier=0.6
 concurrencySet [&Athena_GliderOpenClose_Conc]
}
