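# Sound cue for the vocal call of the Pterodactyl glider in Battle Royale cosmetics.
# A single wave player plays one vocal sound wave at reduced volume, with no randomization or modulation.
# It uses the player glider thrust attenuation, a glider landing sound class, and the glider open/close concurrency group to limit overlapping instances.
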SoundCue /BRCosmetics/Sounds/Gliders/Athena_Glider_Pterodactyl_Vocal_Cue
root=WavePlayer_1
nodes [{id=WavePlayer_1, properties {soundWaveAssetPtr=&/Athena_Glider_Pterodactyl_Vocal01}}]
properties {
 attenuationSettings=&Fort_Player_Glider_Thrust_Athena_Att
 maxDistance=12500
 soundClassObject=&Fort_Gliders_Land
 volumeMultiplier=0.5
 concurrencySet [&Athena_GliderOpenClose_Conc]
}
