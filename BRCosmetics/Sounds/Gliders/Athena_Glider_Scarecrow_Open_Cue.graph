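# Sound cue for the Scarecrow glider's opening in Athena, routed through the shared glider open/close attenuation, concurrency and cosmetic property sheet.
# A mixer layers the main open sound, played immediately at full volume, with two plant-growth waves at reduced volume.
# Each growth layer sits behind its own fixed delay, so they trail the opening by about two seconds, slightly staggered from each other.
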SoundCue /BRCosmetics/Sounds/Gliders/Athena_Glider_Scarecrow_Open_Cue
root=Mixer_0
nodes [
 {id=Delay_0, children [WavePlayer_0], properties {delayMax=2.08, delayMin=2.08}}
 {id=Delay_1, children [WavePlayer_1], properties {delayMax=2.3, delayMin=2.3}}
 {id=Mixer_0, children [WavePlayer_2,Delay_0,Delay_1], properties {inputVolume [1,0.5,0.4]}}
 {id=WavePlayer_0, properties {soundWaveAssetPtr=&/Athena_Glider_Scarecrow_PlantGrow01}}
 {id=WavePlayer_1, properties {soundWaveAssetPtr=&/Athena_Glider_Scarecrow_PlantGrow02}}
 {id=WavePlayer_2, properties {soundWaveAssetPtr=&/Athena_Glider_Scarecrow_Open}}
]
properties {
 attenuationSettings=/Game/Sounds/Attenuation/Fort_Player_Glider_OpenClose_Athena_Att
 audioPropertiesSheet=/BRCosmetics/Sounds/Settings/PropertySheets/Glider/APS_Cosmetic_Glider_Open
 maxDistance=10000
 soundClassObject=&Fort_Gliders_Land
 volumeMultiplier=1
 concurrencySet [&Athena_GliderOpenClose_Conc]
}
