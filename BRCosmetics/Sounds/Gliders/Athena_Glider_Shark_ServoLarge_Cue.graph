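# Sound cue for the large servo sound of the Shark glider cosmetic.
# A Random node picks between two servo waves with equal weight and feeds a Modulator left at default settings for variation on each play.
# Output plays at reduced volume through the player glider thrust attenuation and is routed to the glider landing sound class.
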SoundCue /BRCosmetics/Sounds/Gliders/Athena_Glider_Shark_ServoLarge_Cue
root=Modulator_0
nodes [
 {id=Modulator_0, children [Random_0]}
 {id=Random_0, children [WavePlayer_0,WavePlayer_2], properties {weights [1,1]}}
 {id=WavePlayer_0, properties {soundWaveAssetPtr=&/Athena_Glider_Shark_Servo_11}}
 {id=WavePlayer_2, properties {soundWaveAssetPtr=&/Athena_Glider_Shark_Servo_07}}
]
properties {
 attenuationSettings=&Fort_Player_Glider_Thrust_Athena_Att
 maxDistance=12500
 soundClassObject=&Fort_Gliders_Land
 volumeMultiplier=0.2
}
