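# Sound cue for small servo sounds on the Shark glider cosmetic.
# A Random node picks one of four servo wave variations with equal weight and feeds it through a Modulator at the root.
# It plays at reduced volume, uses the player glider thrust attenuation asset and is routed to the glider landing sound class.
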SoundCue /BRCosmetics/Sounds/Gliders/Athena_Glider_Shark_ServoSmall_Cue
root=Modulator_0
nodes [
 {id=Modulator_0, children [Random_0]}
 {id=Random_0, children [WavePlayer_5,WavePlayer_6,WavePlayer_7,WavePlayer_8], properties {weights [1,1,1,1]}}
 {id=WavePlayer_5, properties {soundWaveAssetPtr=&/Athena_Glider_Shark_Servo_03}}
 {id=WavePlayer_6, properties {soundWaveAssetPtr=&/Athena_Glider_Shark_Servo_04}}
 {id=WavePlayer_7, properties {soundWaveAssetPtr=&/Athena_Glider_Shark_Servo_05}}
 {id=WavePlayer_8, properties {soundWaveAssetPtr=&/Athena_Glider_Shark_Servo_06}}
]
properties {
 attenuationSettings=&Fort_Player_Glider_Thrust_Athena_Att
 maxDistance=12500
 soundClassObject=&Fort_Gliders_Land
 volumeMultiplier=0.2
}
