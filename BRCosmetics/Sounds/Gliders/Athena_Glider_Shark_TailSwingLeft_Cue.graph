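# Sound cue for the left tail swing of the Shark glider cosmetic.
# A Random node picks between two tail-swing wave variants with equal weight and plays it at low volume.
# It uses the player glider thrust attenuation asset and is routed to the glider land sound class.
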SoundCue /BRCosmetics/Sounds/Gliders/Athena_Glider_Shark_TailSwingLeft_Cue
root=Random_0
nodes [
 {id=Random_0, children [WavePlayer_0,WavePlayer_2], properties {weights [1,1]}}
 {id=WavePlayer_0, properties {soundWaveAssetPtr=&/Athena_Glider_Shark_TailSwing_10}}
 {id=WavePlayer_2, properties {soundWaveAssetPtr=&/Athena_Glider_Shark_TailSwing_08}}
]
properties {
 attenuationSettings=&Fort_Player_Glider_Thrust_Athena_Att
 maxDistance=12500
 soundClassObject=&Fort_Gliders_Land
 volumeMultiplier=0.15
}
