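# Sound cue for the right tail swing of the Shark glider in Battle Royale.
# A single wave player feeds the TailSwing_09 wave straight to the output, with no randomization or modulation.
# It plays at a low volume multiplier in the glider land sound class, using the shared Athena player glider thrust attenuation.
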SoundCue /BRCosmetics/Sounds/Gliders/Athena_Glider_Shark_TailSwingRight_Cue
root=WavePlayer_0
nodes [{id=WavePlayer_0, properties {soundWaveAssetPtr=&/Athena_Glider_Shark_TailSwing_09}}]
properties {
 attenuationSettings=&Fort_Player_Glider_Thrust_Athena_Att
 maxDistance=12500
 soundClassObject=&Fort_Gliders_Land
 volumeMultiplier=0.15
}
