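# Sound cue for the Shuttle glider's thrust start, pairing a start sound with a looping thrust bed.
# A mixer plays the start wave and the looping wave together.
# An enveloper ramps the loop's volume up from near silence, so the sustained thrust swells in beneath the one-shot.
# The cue uses glider thrust attenuation and concurrency, the glider landing sound class, and a reduced overall volume.
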SoundCue /BRCosmetics/Sounds/Gliders/Athena_Glider_Shuttle_ThrustStart_Cue
root=Mixer_0
nodes [
 {
  id=Enveloper_0
  children [WavePlayer_1]
  properties {pitchCurve {editorCurveData {keys [{time=0}]}}, volumeCurve {editorCurveData {keys [<at 0 0.1>,<at 1 1>]}}}
 }
 {id=Mixer_0, children [WavePlayer_2,Enveloper_0], properties {inputVolume [1,1]}}
 {id=WavePlayer_1, properties {bLooping=true, soundWaveAssetPtr=/Game/Sounds/Fort_Gadgets/Athena/SFX_Glider_Shuttle_ThrustLoop}}
 {id=WavePlayer_2, properties {soundWaveAssetPtr=&/SFX_Glider_Shuttle_ThrustStart}}
]
properties {
 attenuationSettings=&Fort_Player_Glider_Thrust_Athena_Att
 maxDistance=12500
 soundClassObject=&Fort_Gliders_Land
 volumeMultiplier=0.4
 concurrencySet [&Athena_Glider_Thrust_Conc]
}
