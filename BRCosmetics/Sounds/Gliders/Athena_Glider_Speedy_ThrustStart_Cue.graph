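# Sound cue for the thrust-start sound of the Speedy glider in Athena.
# A single wave player feeds a modulator that slightly randomizes pitch downward on each play.
# It routes through a glider sound class with thrust attenuation and a thrust concurrency set, at reduced volume.
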SoundCue /BRCosmetics/Sounds/Gliders/Athena_Glider_Speedy_ThrustStart_Cue
root=Modulator_0
nodes [
 {id=Modulator_0, children [WavePlayer_3], properties {pitchMax=1, pitchMin=0.9}}
 {id=WavePlayer_3, properties {soundWaveAssetPtr=&/SFX_Glider_Speedy_ThrustStart}}
]
properties {
 attenuationSettings=&Fort_Player_Glider_Thrust_Athena_Att
 maxDistance=12500
 soundClassObject=&Fort_Gliders_Land
 volumeMultiplier=0.35
 concurrencySet [&Athena_Glider_Thrust_Conc]
}
