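# Sound cue for the looping thrust audio of the Speedy glider cosmetic.
# A looping wave player feeds an Enveloper that ramps volume up from partial to full over under a second as the loop starts.
# The cue plays pitched up, with glider thrust attenuation, a cosmetic glider loop properties sheet and the glider sound class.
# Concurrency allows one instance per owner, stopping the oldest.
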SoundCue /BRCosmetics/Sounds/Gliders/Athena_Glider_Speedy_Thrust_Loop_Cue
root=Enveloper_0
nodes [
 {
  id=Enveloper_0
  children [WavePlayer_0]
  properties {pitchCurve {editorCurveData {keys [{time=0}]}}, volumeCurve {editorCurveData {keys [<at 0 0.3>,<at 0.7 1>]}}}
 }
 {id=WavePlayer_0, properties {bLooping=true, soundWaveAssetPtr=/Game/Sounds/Fort_Gadgets/Athena/SFX_Glider_Speedy_Thrust_Loop}}
]
properties {
 attenuationSettings=&Fort_Player_Glider_Thrust_Athena_Att
 audioPropertiesSheet=&APS_Cosmetic_Glider_Loop
 maxDistance=12500
 pitchMultiplier=1.5
 soundClassObject=&Fort_Gliders_Land
 volumeMultiplier=1
 concurrencySet [/Game/Sounds/Concurrency_Rules/OneMaximum_Owner_StopOldest]
}
