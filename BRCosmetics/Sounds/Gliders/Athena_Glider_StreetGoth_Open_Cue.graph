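# Sound cue for the opening of the StreetGoth cosmetic glider.
# A mixer layers the main open wave with a wood open wave that is quieter, pitched slightly down and delayed by a fixed time so it trails the first.
# Attenuation, concurrency and the property sheet come from shared glider assets, and it routes to the glider land sound class.
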SoundCue /BRCosmetics/Sounds/Gliders/Athena_Glider_StreetGoth_Open_Cue
root=Mixer_1
nodes [
 {id=Delay_2, children [WavePlayer_1], properties {delayMax=1.45, delayMin=1.45}}
 {id=Mixer_1, children [WavePlayer_0,Modulator_1], properties {inputVolume [1,0.67]}}
 {id=Modulator_1, children [Delay_2], properties {pitchMax=0.9, pitchMin=0.9}}
 {id=WavePlayer_0, properties {soundWaveAssetPtr=&/Athena_Glider_StreetGoth_Open01}}
 {id=WavePlayer_1, properties {soundWaveAssetPtr=&/Athena_Glider_StreetGoth_OpenWood01}}
]
properties {
 attenuationSettings=/Game/Sounds/Attenuation/Fort_Player_Glider_OpenClose_Athena_Att
 audioPropertiesSheet=/BRCosmetics/Sounds/Settings/PropertySheets/Glider/APS_Cosmetic_Glider_Open
 maxDistance=10000
 soundClassObject=&Fort_Gliders_Land
 volumeMultiplier=1.1
 concurrencySet [&Athena_GliderOpenClose_Conc]
}
