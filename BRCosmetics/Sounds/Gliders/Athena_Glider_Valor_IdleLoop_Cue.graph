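# Looping in-air idle sound cue for the Valor glider cosmetic, with glider loop attenuation, sound class and audio properties sheet.
# A quality-level switch feeds the standard looping wave to its first two inputs and the "opt" variant to its last two.
# Each wave passes through an enveloper whose volume curve rises to full over about a second.
# Owner-limited concurrency allows one instance and prevents new ones.
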SoundCue /BRCosmetics/Sounds/Gliders/Athena_Glider_Valor_IdleLoop_Cue
root=QualityLevel_0
nodes [
 {
  id=Enveloper_0
  children [WavePlayer_2]
  properties {pitchCurve {editorCurveData {keys [{time=0}]}}, volumeCurve {editorCurveData {keys [<at 0.004 0.0455>,<at 1 1>]}}}
 }
 {
  id=Enveloper_1
  children [WavePlayer_0]
  properties {pitchCurve {editorCurveData {keys [{time=0}]}}, volumeCurve {editorCurveData {keys [<at 0.004 0.0455>,<at 1 1>]}}}
 }
 {id=QualityLevel_0, children [Enveloper_0,Enveloper_0,Enveloper_1,Enveloper_1]}
 {id=WavePlayer_0, properties {bLooping=true, soundWaveAssetPtr=&/SFX_Glider_Valor_InAirLoop_01_opt}}
 {id=WavePlayer_2, properties {bLooping=true, soundWaveAssetPtr=&/SFX_Glider_Valor_InAirLoop_01}}
]
properties {
 attenuationSettings=&Fort_Player_Glider_Thrust_Athena_Att
 audioPropertiesSheet=&APS_Cosmetic_Glider_Loop
 bOverrideConcurrency=true
 maxDistance=12500
 soundClassObject=&Fort_Gliders_Land
 volumeMultiplier=0.2
 concurrencyOverrides {bLimitToOwner=true, maxCount=1, resolutionRule=PreventNew}
}
