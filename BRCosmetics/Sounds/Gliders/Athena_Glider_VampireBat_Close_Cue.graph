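# Sound cue for the VampireBat cosmetic glider's close sound in Athena, played when the glider closes.
# A mixer layers the main close wave at full volume with a quieter close vocal wave.
# The cue routes through the glider landing sound class, uses shared glider open/close attenuation and concurrency, and takes the cosmetic glider-close audio property sheet.
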SoundCue /BRCosmetics/Sounds/Gliders/Athena_Glider_VampireBat_Close_Cue
root=Mixer_0
nodes [
 {id=Mixer_0, children [WavePlayer_1,WavePlayer_2], properties {inputVolume [1,0.2]}}
 {id=WavePlayer_1, properties {soundWaveAssetPtr=&/Athena_Glider_VampireBat_Close}}
 {id=WavePlayer_2, properties {soundWaveAssetPtr=&/Athena_Glider_VampireBat_CloseVocal_01}}
]
properties {
 attenuationSettings=/Game/Sounds/Attenuation/Fort_Player_Glider_OpenClose_Athena_Att
 audioPropertiesSheet=/BRCosmetics/Sounds/Settings/PropertySheets/Glider/APS_Cosmetic_Glider_Close
 maxDistance=10000
 soundClassObject=&Fort_Gliders_Land
 volumeMultiplier=0.7
 concurrencySet [&Athena_GliderOpenClose_Conc]
}
