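# Sound cue for the Vampire Bat glider's loop audio, pairing a continuous loop with intermittent bat vocals.
# A mixer combines a slightly pitched-down looping wave with a repeating branch that waits a random delay, then plays one of three weighted-random vocals through a modulator.
# Glider thrust attenuation, a cosmetic glider loop properties sheet and a glider sound class govern playback.
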SoundCue /BRCosmetics/Sounds/Gliders/Athena_Glider_VampireBat_Loop_Cue
root=Mixer_0
nodes [
 {id=Delay_0, children [Modulator_0], properties {delayMax=9, delayMin=5}}
 {id=Looping_0, children [Delay_0]}
 {id=Mixer_0, children [Modulator_1,Looping_0], properties {inputVolume [0.7,0.75]}}
 {id=Modulator_0, children [Random_0]}
 {id=Modulator_1, children [WavePlayer_0], properties {pitchMax=0.9, pitchMin=0.9}}
 {id=Random_0, children [WavePlayer_2,WavePlayer_4,WavePlayer_5], properties {weights [2,1,2]}}
 {id=WavePlayer_0, properties {bLooping=true, soundWaveAssetPtr=&/Athena_Glider_VampireBat_Loop_02}}
 {id=WavePlayer_2, properties {soundWaveAssetPtr=&/Athena_Glider_VampireBat_Vocal_01}}
 {id=WavePlayer_4, properties {soundWaveAssetPtr=&/Athena_Glider_VampireBat_Vocal_03}}
 {id=WavePlayer_5, properties {soundWaveAssetPtr=&/Athena_Glider_VampireBat_Vocal_04}}
]
properties {
 attenuationSettings=&Fort_Player_Glider_Thrust_Athena_Att
 audioPropertiesSheet=&APS_Cosmetic_Glider_Loop
 maxDistance=12500
 soundClassObject=&Fort_Gliders_Land
 volumeMultiplier=0.45
}
