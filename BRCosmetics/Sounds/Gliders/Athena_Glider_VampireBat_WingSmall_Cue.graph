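# Sound cue for the small wing sound of the Vampire Bat glider cosmetic.
# A Random node picks one of three equally weighted wave variations and feeds a Modulator that randomizes pitch from normal upward for extra variety.
# Playback uses the shared player glider thrust attenuation and glider landing sound class, with a concurrency override capping simultaneous instances.
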SoundCue /BRCosmetics/Sounds/Gliders/Athena_Glider_VampireBat_WingSmall_Cue
root=Modulator_0
nodes [
 {id=Modulator_0, children [Random_0], properties {pitchMax=1.4, pitchMin=1}}
 {id=Random_0, children [WavePlayer_1,WavePlayer_2,WavePlayer_3], properties {weights [1,1,1]}}
 {id=WavePlayer_1, properties {soundWaveAssetPtr=&/Athena_Glider_VampireBat_WingSmall_01}}
 {id=WavePlayer_2, properties {soundWaveAssetPtr=&/Athena_Glider_VampireBat_WingSmall_02}}
 {id=WavePlayer_3, properties {soundWaveAssetPtr=&/Athena_Glider_VampireBat_WingSmall_03}}
]
properties {
 attenuationSettings=&Fort_Player_Glider_Thrust_Athena_Att
 bOverrideConcurrency=true
 maxDistance=12500
 soundClassObject=&Fort_Gliders_Land
 volumeMultiplier=0.8
 concurrencyOverrides {maxCount=5}
}
