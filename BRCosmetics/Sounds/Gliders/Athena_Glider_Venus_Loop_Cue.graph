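# Looping sound cue for the Venus glider cosmetic in Athena, playing the glider's sustained loop.
# A single looping wave player feeds the output directly, with no randomization, mixing or modulation nodes.
# It uses the glider thrust attenuation, cosmetic glider loop audio properties sheet and glider sound class at reduced volume.
# Playback shares the glider open/close concurrency group.
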SoundCue /BRCosmetics/Sounds/Gliders/Athena_Glider_Venus_Loop_Cue
root=WavePlayer_2
nodes [{id=WavePlayer_2, properties {bLooping=true, soundWaveAssetPtr=&/Athena_Glider_Venus_Loop_02}}]
properties {
 attenuationSettings=&Fort_Player_Glider_Thrust_Athena_Att
 audioPropertiesSheet=&APS_Cosmetic_Glider_Loop
 maxDistance=12500
 soundClassObject=&Fort_Gliders_Land
 volumeMultiplier=0.5
 concurrencySet [&Athena_GliderOpenClose_Conc]
}
