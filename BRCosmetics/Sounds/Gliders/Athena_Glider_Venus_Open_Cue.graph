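# Sound cue for the opening of the Venus cosmetic glider, routed through glider sound class, open/close attenuation and concurrency settings.
# A mixer layers the Open wave at full volume with a quieter Extend wave.
# The Extend layer passes through a fixed delay and a modulator that slightly lowers its pitch, so it follows the initial open sound.
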SoundCue /BRCosmetics/Sounds/Gliders/Athena_Glider_Venus_Open_Cue
root=Mixer_0
nodes [
 {id=Delay_0, children [WavePlayer_4], properties {delayMax=0.8, delayMin=0.8}}
 {id=Mixer_0, children [WavePlayer_0,Modulator_0], properties {inputVolume [1,0.4]}}
 {id=Modulator_0, children [Delay_0], properties {pitchMax=0.9, pitchMin=0.9}}
 {id=WavePlayer_0, properties {soundWaveAssetPtr=&/Athena_Glider_Venus_Open_01}}
 {id=WavePlayer_4, properties {soundWaveAssetPtr=&/Athena_Glider_Venus_Extend_01}}
]
properties {
 attenuationSettings=/Game/Sounds/Attenuation/Fort_Player_Glider_OpenClose_Athena_Att
 audioPropertiesSheet=/BRCosmetics/Sounds/Settings/PropertySheets/Glider/APS_Cosmetic_Glider_Open
 maxDistance=10000
 soundClassObject=&Fort_Gliders_Land
 volumeMultiplier=0.7
 concurrencySet [&Athena_GliderOpenClose_Conc]
}
