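# Sound cue for the DarkNinja glider's acceleration start sound, part of the Battle Royale cosmetics content.
# A Random node picks with equal weight between two wave variants and plays the chosen one.
# It uses the player glider thrust attenuation asset and is routed to the Fort_Gliders_Land sound class.
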SoundCue /BRCosmetics/Sounds/Gliders/Glider_DarkNinja_AccelStart_Cue
root=Random_0
nodes [
 {id=Random_0, children [WavePlayer_1,WavePlayer_2], properties {weights [1,1]}}
 {id=WavePlayer_1, properties {soundWaveAssetPtr=&/Glider_DarkNinja_AccelStart_01}}
 {id=WavePlayer_2, properties {soundWaveAssetPtr=&/Glider_DarkNinja_AccelStart_02}}
]
properties {
 attenuationSettings=&Fort_Player_Glider_Thrust_Athena_Att
 maxDistance=12500
 soundClassObject=&Fort_Gliders_Land
 volumeMultiplier=1
}
