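# Sound cue for the thrust-end sound of the GarageBand glider cosmetic.
# A single wave player is the root and plays one wave asset, with no randomization or mixing.
# The cue runs at reduced volume with glider thrust attenuation and concurrency, the glider landing sound class, and a glider loop audio properties sheet.
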SoundCue /BRCosmetics/Sounds/Gliders/Glider_GarageBand_ThrustEnd_Cue
root=WavePlayer_0
nodes [{id=WavePlayer_0, properties {soundWaveAssetPtr=&/Glider_GarageBand_ThrustEnd_01}}]
properties {
 attenuationSettings=&Fort_Player_Glider_Thrust_Athena_Att
 audioPropertiesSheet=&APS_Cosmetic_Glider_Loop
 maxDistance=12500
 soundClassObject=&Fort_Gliders_Land
 volumeMultiplier=0.5
 concurrencySet [&Athena_Glider_Thrust_Conc]
}
