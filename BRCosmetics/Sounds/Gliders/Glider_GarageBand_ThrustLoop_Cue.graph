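# Looping sound cue for the GarageBand glider cosmetic, layering a music loop with a thrust loop.
# A Mixer blends the music at full volume with the slightly quieter thrust wave.
# A quality-level switch picks either the standard or "_opt" music wave, while thrust is identical at every level.
# The cue uses the player glider thrust attenuation and the gliders music sound class.
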SoundCue /BRCosmetics/Sounds/Gliders/Glider_GarageBand_ThrustLoop_Cue
root=Mixer_0
nodes [
 {id=Mixer_0, children [QualityLevel_0, WavePlayer_0], properties {inputVolume [1,0.75]}}
 {id=QualityLevel_0, children [WavePlayer_1,WavePlayer_1,WavePlayer_2,WavePlayer_2]}
 {id=WavePlayer_0, properties {bLooping=true, soundWaveAssetPtr=&/Glider_GarageBand_ThrustLoop_01}}
 {id=WavePlayer_1, properties {bLooping=true, soundWaveAssetPtr=&/Glider_GarageBand_Music_Loop_01}}
 {id=WavePlayer_2, properties {bLooping=true, soundWaveAssetPtr=&/Glider_GarageBand_Music_Loop_01_opt}}
]
properties {
 attenuationSettings=&Fort_Player_Glider_Thrust_Athena_Att
 audioPropertiesSheet=&APS_Cosmetic_Glider_Loop
 maxDistance=12500
 soundClassObject=/Game/Sounds/FortSoundClassesAndModes/Fort_Gliders_Music
 volumeMultiplier=1
}
