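# Sound cue for the turn signal sound of the GumShoe glider cosmetic.
# A single wave player feeds the TurnSignal wave straight to the root at reduced volume.
# It uses the Athena player glider thrust attenuation and the glider land sound class.
# A concurrency override allows one instance per owner and prevents new plays while one is active.
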SoundCue /BRCosmetics/Sounds/Gliders/Glider_GumShoe_TurnSignal_Cue
root=WavePlayer_0
nodes [{id=WavePlayer_0, properties {soundWaveAssetPtr=&/Glider_GumShoe_TurnSignal}}]
properties {
 attenuationSettings=&Fort_Player_Glider_Thrust_Athena_Att
 bOverrideConcurrency=true
 maxDistance=12500
 soundClassObject=&Fort_Gliders_Land
 volumeMultiplier=0.2
 concurrencyOverrides {bLimitToOwner=true, maxCount=1, resolutionRule=PreventNew}
}
