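# Looping SoundCue for the Hacivat cosmetic glider, combining its music loop with its operating loop.
# A Mixer plays two looping Wave Players at once, with the music at reduced input volume and the operate loop at full.
# The cue raises pitch and uses player glider thrust attenuation, a cosmetic glider loop properties sheet and the gliders music sound class.
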SoundCue /BRCosmetics/Sounds/Gliders/Glider_Hacivat_MusicLoop_Cue
root=Mixer_0
nodes [
 {id=Mixer_0, children [WavePlayer_0,WavePlayer_1], properties {inputVolume [0.5,1]}}
 {id=WavePlayer_0, properties {bLooping=true, soundWaveAssetPtr=&/Glider_Hacivat_MusicLoop}}
 {id=WavePlayer_1, properties {bLooping=true, soundWaveAssetPtr=&/Glider_Hacivat_OperateLoop}}
]
properties {
 attenuationSettings=&Fort_Player_Glider_Thrust_Athena_Att
 audioPropertiesSheet=&APS_Cosmetic_Glider_Loop
 maxDistance=12500
 pitchMultiplier=1.2
 soundClassObject=/Game/Sounds/FortSoundClassesAndModes/Fort_Gliders_Music
 volumeMultiplier=1
}
