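# Sound cue for the thrust start sound of the Ice Maiden glider.
# A single wave player feeds a modulator that slightly randomizes volume on each play.
# The cue plays at a low overall volume with glider thrust attenuation, a glider sound class and a shared thrust concurrency limit.
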SoundCue /BRCosmetics/Sounds/Gliders/Glider_IceMaiden_Thrust_Start_Cue
root=Modulator_0
nodes [
 {id=Modulator_0, children [WavePlayer_1], properties {volumeMax=1, volumeMin=0.85}}
 {id=WavePlayer_1, properties {soundWaveAssetPtr=&/Glider_IceMaiden_Thrust_Start}}
]
properties {
 attenuationSettings=&Fort_Player_Glider_Thrust_Athena_Att
 maxDistance=12500
 soundClassObject=&Fort_Gliders_Land
 volumeMultiplier=0.15
 concurrencySet [&Athena_Glider_Thrust_Conc]
}
