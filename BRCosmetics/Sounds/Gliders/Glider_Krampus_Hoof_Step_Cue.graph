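# Sound cue for hoof step sounds on the Krampus glider, classed as an in-air glider sound.
# A random node picks one of six equally weighted hoof waves, and a modulator slightly varies its volume on each play.
# The cue plays at reduced volume with glider thrust attenuation and overrides concurrency to cap simultaneous instances at a small count.
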SoundCue /BRCosmetics/Sounds/Gliders/Glider_Krampus_Hoof_Step_Cue
root=Modulator_0
nodes [
 {id=Modulator_0, children [Random_0], properties {volumeMax=1, volumeMin=0.85}}
 {id=Random_0, children [WavePlayer_1,WavePlayer_2,WavePlayer_3,WavePlayer_4,WavePlayer_5,WavePlayer_6], properties {weights [1,1,1,1,1,1]}}
 {id=WavePlayer_1, properties {soundWaveAssetPtr=&/Glider_Krampus_Hoof_01}}
 {id=WavePlayer_2, properties {soundWaveAssetPtr=&/Glider_Krampus_Hoof_02}}
 {id=WavePlayer_3, properties {soundWaveAssetPtr=&/Glider_Krampus_Hoof_03}}
 {id=WavePlayer_4, properties {soundWaveAssetPtr=&/Glider_Krampus_Hoof_04}}
 {id=WavePlayer_5, properties {soundWaveAssetPtr=&/Glider_Krampus_Hoof_05}}
 {id=WavePlayer_6, properties {soundWaveAssetPtr=&/Glider_Krampus_Hoof_06}}
]
properties {
 attenuationSettings=&Fort_Player_Glider_Thrust_Athena_Att
 bOverrideConcurrency=true
 maxDistance=12500
 soundClassObject=&Fort_Gliders_InAir
 volumeMultiplier=0.3
 concurrencyOverrides {maxCount=4}
}
