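# Sound cue for the opening sound of the Krampus glider cosmetic.
# A mixer plays the main open wave alongside a branch delayed by a fixed 1.4 seconds.
# That branch uses a weighted random node with an empty second slot, so a goat sound plays as a rare easter egg, about one percent of the time.
# Glider-specific attenuation, concurrency, sound class and property sheet govern playback.
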SoundCue /BRCosmetics/Sounds/Gliders/Glider_Krampus_Open_Cue
root=Mixer_0
nodes [
 {id=Delay_0, children [Random_0], properties {delayMax=1.4, delayMin=1.4}}
 {id=Mixer_0, children [Delay_0,WavePlayer_1], properties {inputVolume [1,1]}}
 {
  id=Random_0
  children [WavePlayer_2,null]
  properties {bRandomizeWithoutReplacement=false, bShouldExcludeFromBranchCulling=true, weights [0.01,0.99]}
 }
 {id=WavePlayer_1, properties {soundWaveAssetPtr=&/Glider_Krampus_Open}}
 {id=WavePlayer_2, properties {soundWaveAssetPtr=&/Glider_Krampus_Open_Goat}}
]
properties {
 attenuationSettings=/Game/Sounds/Attenuation/Fort_Player_Glider_OpenClose_Athena_Att
 audioPropertiesSheet=/BRCosmetics/Sounds/Settings/PropertySheets/Glider/APS_Cosmetic_Glider_Open
 maxDistance=10000
 soundClassObject=&Fort_Gliders_Land
 volumeMultiplier=1
 concurrencySet [&Athena_GliderOpenClose_Conc]
}
