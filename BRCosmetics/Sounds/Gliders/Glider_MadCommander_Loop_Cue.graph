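# Sound cue for the looping thrust sound of the MadCommander glider cosmetic.
# A single looping wave player is the root and plays the thrust loop wave, with no randomization, mixing or modulation nodes.
# The cue uses the Athena player glider thrust attenuation, the cosmetic glider loop audio properties sheet and the Fort_Gliders_Land sound class.
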SoundCue /BRCosmetics/Sounds/Gliders/Glider_MadCommander_Loop_Cue
root=WavePlayer_1
nodes [{id=WavePlayer_1, properties {bLooping=true, soundWaveAssetPtr=&/SFX_Glider_MadCommander_Thrust_Loop}}]
properties {
 attenuationSettings=&Fort_Player_Glider_Thrust_Athena_Att
 audioPropertiesSheet=&APS_Cosmetic_Glider_Loop
 maxDistance=12500
 soundClassObject=&Fort_Gliders_Land
}
