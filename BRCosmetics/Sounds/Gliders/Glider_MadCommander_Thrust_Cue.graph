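# Sound cue for the looping thrust audio of the Mad Commander glider cosmetic.
# A single wave player feeds a modulator that adds a very slight random pitch variation.
# The cue plays at reduced volume through the glider thrust attenuation, glider sound class, cosmetic loop audio properties sheet and a shared glider thrust concurrency group.
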SoundCue /BRCosmetics/Sounds/Gliders/Glider_MadCommander_Thrust_Cue
root=Modulator_0
nodes [
 {id=Modulator_0, children [WavePlayer_2], properties {pitchMax=1.01}}
 {id=WavePlayer_2, properties {soundWaveAssetPtr=&/SFX_Glider_MadCommander_Thrust_02}}
]
properties {
 attenuationSettings=&Fort_Player_Glider_Thrust_Athena_Att
 audioPropertiesSheet=&APS_Cosmetic_Glider_Loop
 maxDistance=12500
 soundClassObject=&Fort_Gliders_Land
 volumeMultiplier=0.1
 concurrencySet [&Athena_Glider_Thrust_Conc]
}
