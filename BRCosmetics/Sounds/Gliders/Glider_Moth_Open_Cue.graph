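# Sound cue for the Moth glider's open sound in the battle royale cosmetics.
# A mixer layers the Moth open wave with a quieter shuttle thrust-start wave, which a fixed delay pushes two seconds later.
# Playback uses the shared glider open/close attenuation and concurrency, a glider-open property sheet, and the glider landing sound class.
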SoundCue /BRCosmetics/Sounds/Gliders/Glider_Moth_Open_Cue
root=Mixer_0
nodes [
 {id=Delay_0, children [WavePlayer_2], properties {delayMax=2, delayMin=2}}
 {id=Mixer_0, children [WavePlayer_0,Delay_0], properties {inputVolume [1,0.65]}}
 {id=WavePlayer_0, properties {soundWaveAssetPtr=&/Glider_Moth_Open_02}}
 {id=WavePlayer_2, properties {soundWaveAssetPtr=&/SFX_Glider_Shuttle_ThrustStart}}
]
properties {
 attenuationSettings=/Game/Sounds/Attenuation/Fort_Player_Glider_OpenClose_Athena_Att
 audioPropertiesSheet=/BRCosmetics/Sounds/Settings/PropertySheets/Glider/APS_Cosmetic_Glider_Open
 maxDistance=10000
 soundClassObject=&Fort_Gliders_Land
 volumeMultiplier=0.85
 concurrencySet [&Athena_GliderOpenClose_Conc]
}
