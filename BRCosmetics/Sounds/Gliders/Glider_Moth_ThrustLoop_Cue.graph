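# Sound cue for the Moth glider's thrust, layering a one-shot start sound with looping thrust audio.
# The start layer reuses the Shuttle glider's ThrustStart wave with slight random pitch variation.
# An enveloper quickly fades in a mix of two looping Moth waves, the first randomly pitched up and the second weighted louder.
# Glider thrust attenuation and a cosmetic glider loop audio properties sheet govern playback.
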SoundCue /BRCosmetics/Sounds/Gliders/Glider_Moth_ThrustLoop_Cue
root=Mixer_2
nodes [
 {
  id=Enveloper_0
  children [Mixer_1]
  properties {pitchCurve {editorCurveData {keys [{time=0}]}}, volumeCurve {editorCurveData {keys [<at 0.0088 0.0072>,<at 0.1502 0.9997>]}}}
 }
 {id=Mixer_1, children [Modulator_1,WavePlayer_1], properties {inputVolume [0.9,2.9]}}
 {id=Mixer_2, children [Modulator_0,Enveloper_0], properties {inputVolume [0.9,1]}}
 {id=Modulator_0, children [WavePlayer_2], properties {pitchMax=1.15, pitchMin=1.05}}
 {id=Modulator_1, children [WavePlayer_0], properties {pitchMax=1.3, pitchMin=1.2}}
 {id=WavePlayer_0, properties {bLooping=true, soundWaveAssetPtr=&/Glider_Moth_ThrustLoop_01}}
 {id=WavePlayer_1, properties {bLooping=true, soundWaveAssetPtr=&/Glider_Moth_ThrustLoop_02}}
 {id=WavePlayer_2, properties {soundWaveAssetPtr=&/SFX_Glider_Shuttle_ThrustStart}}
]
properties {
 attenuationSettings=&Fort_Player_Glider_Thrust_Athena_Att
 audioPropertiesSheet=&APS_Cosmetic_Glider_Loop
 maxDistance=12500
 pitchMultiplier=1.1
 soundClassObject=&Fort_Gliders_Land
 volumeMultiplier=0.5
}
