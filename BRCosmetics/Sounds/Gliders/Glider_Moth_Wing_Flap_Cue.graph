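# Sound cue for the wing flap of the Moth glider.
# A random node picks between two equally weighted flap waves and feeds a modulator that varies pitch, on top of lowered overall pitch and volume.
# It uses glider thrust attenuation and the gliders landing sound class.
# Overridden concurrency limits it to two instances per owner, stopping the oldest, in place of the assigned glider open/close concurrency set.
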SoundCue /BRCosmetics/Sounds/Gliders/Glider_Moth_Wing_Flap_Cue
root=Modulator_0
nodes [
 {id=Modulator_0, children [Random_2], properties {pitchMax=1.1, volumeMin=1}}
 {id=Random_2, children [WavePlayer_6,WavePlayer_7], properties {weights [1,1]}}
 {id=WavePlayer_6, properties {soundWaveAssetPtr=&/Glider_Moth_Flap_03}}
 {id=WavePlayer_7, properties {soundWaveAssetPtr=&/Glider_Moth_Flap_04}}
]
properties {
 attenuationSettings=&Fort_Player_Glider_Thrust_Athena_Att
 bOverrideConcurrency=true
 maxDistance=12500
 pitchMultiplier=0.8
 soundClassObject=&Fort_Gliders_Land
 volumeMultiplier=0.35
 concurrencyOverrides {bLimitToOwner=true, maxCount=2, resolutionRule=StopOldest}
 concurrencySet [&Athena_GliderOpenClose_Conc]
}
