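# Sound cue for the wing flap of the Muertos glider, layering a flap sound with a quieter sparkle.
# A mixer blends one of three random flap waves with one of three random sparkle waves, which are pitched up by their own modulator.
# An outer modulator varies the combined result.
# Glider thrust attenuation spatializes it, and overridden concurrency limits copies per owner, stopping the oldest.
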SoundCue /BRCosmetics/Sounds/Gliders/Glider_Muertos_Wing_Flap_Cue
root=Modulator_0
nodes [
 {id=Mixer_0, children [Modulator_1,Random_1], properties {inputVolume [0.3,1]}}
 {id=Modulator_0, children [Mixer_0], properties {volumeMin=1}}
 {id=Modulator_1, children [Random_0], properties {pitchMax=1.35, pitchMin=1.1}}
 {id=Random_0, children [WavePlayer_1,WavePlayer_2,WavePlayer_5], properties {weights [1,1,1]}}
 {id=Random_1, children [WavePlayer_0,WavePlayer_3,WavePlayer_4], properties {weights [1,1,1]}}
 {id=WavePlayer_0, properties {soundWaveAssetPtr=&/Glider_Muertos_Flap_01}}
 {id=WavePlayer_1, properties {soundWaveAssetPtr=/Game/Sounds/Fort_Gadgets/Athena/Glider_Muertos_Flap_Sparkle_01}}
 {id=WavePlayer_2, properties {soundWaveAssetPtr=/Game/Sounds/Fort_Gadgets/Athena/Glider_Muertos_Flap_Sparkle_02}}
 {id=WavePlayer_3, properties {soundWaveAssetPtr=&/Glider_Muertos_Flap_02}}
 {id=WavePlayer_4, properties {soundWaveAssetPtr=&/Glider_Muertos_Flap_03}}
 {id=WavePlayer_5, properties {soundWaveAssetPtr=/Game/Sounds/Fort_Gadgets/Athena/Glider_Muertos_Flap_Sparkle_03}}
]
properties {
 attenuationSettings=&Fort_Player_Glider_Thrust_Athena_Att
 bOverrideConcurrency=true
 maxDistance=12500
 pitchMultiplier=0.8
 soundClassObject=&Fort_Gliders_Land
 volumeMultiplier=0.45
 concurrencyOverrides {bLimitToOwner=true, maxCount=2, resolutionRule=StopOldest}
 concurrencySet [&Athena_GliderOpenClose_Conc]
}
