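# Looping thrust sound cue for the Octoberfest glider cosmetic.
# A fixed delay holds off playback, then an Enveloper quickly ramps up the volume of the looping thrust wave.
# It uses the player glider thrust attenuation and cosmetic glider loop property sheet, and is routed through the gliders music sound class.
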SoundCue /BRCosmetics/Sounds/Gliders/Glider_Octoberfest_Loop_Cue
root=Delay_0
nodes [
 {id=Delay_0, children [Enveloper_0], properties {delayMax=2, delayMin=2}}
 {
  id=Enveloper_0
  children [WavePlayer_1]
  properties {pitchCurve {editorCurveData {keys [{time=0}]}}, volumeCurve {editorCurveData {keys [<at 0 0.2>,<at 0.3 1>]}}}
 }
 {id=WavePlayer_1, properties {bLooping=true, soundWaveAssetPtr=&/Glider_Octoberfest_ThrustLoop}}
]
properties {
 attenuationSettings=&Fort_Player_Glider_Thrust_Athena_Att
 audioPropertiesSheet=&APS_Cosmetic_Glider_Loop
 maxDistance=12500
 soundClassObject=/Game/Sounds/FortSoundClassesAndModes/Fort_Gliders_Music
 volumeMultiplier=0.56
}
