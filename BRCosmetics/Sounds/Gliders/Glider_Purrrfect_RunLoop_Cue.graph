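# Looping Sound Cue for the Purrrfect glider cosmetic's run loop.
# A Mixer layers the Purrrfect thrust loop over a much quieter Valor in-air loop borrowed from another glider, which a Modulator shifts to a fixed higher pitch.
# The cue plays at reduced volume and uses glider thrust attenuation, the cosmetic glider loop properties sheet, the glider landing sound class and glider open/close concurrency.
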SoundCue /BRCosmetics/Sounds/Gliders/Glider_Purrrfect_RunLoop_Cue
root=Mixer_0
nodes [
 {id=Mixer_0, children [Modulator_0,WavePlayer_1], properties {inputVolume [0.2,2.2]}}
 {id=Modulator_0, children [WavePlayer_3], properties {pitchMax=1.5, pitchMin=1.5}}
 {id=WavePlayer_1, properties {bLooping=true, soundWaveAssetPtr=&/Glider_Purrrfect_ThrustLoop_01}}
 {id=WavePlayer_3, properties {bLooping=true, soundWaveAssetPtr=&/SFX_Glider_Valor_InAirLoop_01}}
]
properties {
 attenuationSettings=&Fort_Player_Glider_Thrust_Athena_Att
 audioPropertiesSheet=&APS_Cosmetic_Glider_Loop
 maxDistance=12500
 soundClassObject=&Fort_Gliders_Land
 volumeMultiplier=0.5
 concurrencySet [&Athena_GliderOpenClose_Conc]
}
