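# Sound cue for the thrust-start sound of the Purrrfect glider cosmetic.
# A single wave player feeds the output directly, with no randomization or mixing.
# The cue lowers the pitch and boosts the volume of that wave.
# It uses player glider thrust attenuation, the glider landing sound class and a glider thrust concurrency set.
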SoundCue /BRCosmetics/Sounds/Gliders/Glider_Purrrfect_ThrustStart_Cue
root=WavePlayer_0
nodes [{id=WavePlayer_0, properties {soundWaveAssetPtr=&/Glider_Purrrfect_ThrustStart_01}}]
properties {
 attenuationSettings=&Fort_Player_Glider_Thrust_Athena_Att
 maxDistance=12500
 pitchMultiplier=0.5
 soundClassObject=&Fort_Gliders_Land
 volumeMultiplier=1.5
 concurrencySet [&Athena_Glider_Thrust_Conc]
}
