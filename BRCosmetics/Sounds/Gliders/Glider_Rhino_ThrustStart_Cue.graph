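# Sound cue for the thrust-start sound of the Rhino glider cosmetic.
# A single wave player feeds a modulator that varies pitch and volume, with upward pitch variation capped at the original pitch.
# It uses the player glider open/close attenuation and the glider landing sound class, plays at reduced volume, and is limited by the shared glider thrust concurrency.
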SoundCue /BRCosmetics/Sounds/Gliders/Glider_Rhino_ThrustStart_Cue
root=Modulator_0
nodes [
 {id=Modulator_0, children [WavePlayer_0], properties {pitchMax=1}}
 {id=WavePlayer_0, properties {soundWaveAssetPtr=&/SFX_Glider_Rhino_ThrustStart}}
]
properties {
 attenuationSettings=/Game/Sounds/Attenuation/Fort_Player_Glider_OpenClose_Athena_Att
 maxDistance=10000
 soundClassObject=&Fort_Gliders_Land
 volumeMultiplier=0.55
 concurrencySet [&Athena_Glider_Thrust_Conc]
}
