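# Sound cue for the Skull Trooper glider's close sound.
# A mixer layers a full-level close wave with a quieter vocal close wave, both playing together.
# The cue runs at half volume on the glider landing sound class, using the shared glider open/close attenuation and concurrency and a glider-close audio property sheet.
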SoundCue /BRCosmetics/Sounds/Gliders/Glider_SkullTrooper_Close_Cue
root=Mixer_0
nodes [
 {id=Mixer_0, children [WavePlayer_0,WavePlayer_3], properties {inputVolume [1,0.3]}}
 {id=WavePlayer_0, properties {soundWaveAssetPtr=&/Glider_SkullTrooper_Close_01}}
 {id=WavePlayer_3, properties {soundWaveAssetPtr=&/Glider_SkullTrooper_CloseVocal_01}}
]
properties {
 attenuationSettings=/Game/Sounds/Attenuation/Fort_Player_Glider_OpenClose_Athena_Att
 audioPropertiesSheet=/BRCosmetics/Sounds/Settings/PropertySheets/Glider/APS_Cosmetic_Glider_Close
 maxDistance=10000
 soundClassObject=&Fort_Gliders_Land
 volumeMultiplier=0.5
 concurrencySet [&Athena_GliderOpenClose_Conc]
}
