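# Looping sound cue for the Skull Trooper glider, layering a continuous flight loop with intermittent vocal sounds.
# A mixer blends the looping wave, pitched slightly down, with a quieter branch that endlessly picks one of four equally weighted vocals after a random multi-second delay, passed through a modulator.
# Glider attenuation, sound class and cosmetic glider-loop properties sheet govern playback.
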SoundCue /BRCosmetics/Sounds/Gliders/Glider_SkullTrooper_Loop_Cue
root=Mixer_0
nodes [
 {id=Delay_0, children [Modulator_0], properties {delayMax=9, delayMin=4}}
 {id=Looping_0, children [Delay_0]}
 {id=Mixer_0, children [Modulator_1,Looping_0], properties {inputVolume [0.8,0.45]}}
 {id=Modulator_0, children [Random_2]}
 {id=Modulator_1, children [WavePlayer_1], properties {pitchMax=0.9, pitchMin=0.9}}
 {id=Random_2, children [WavePlayer_8, WavePlayer_9, WavePlayer_10, WavePlayer_11], properties {weights [1,1,1,1]}}
 {id=WavePlayer_1, properties {bLooping=true, soundWaveAssetPtr=&/Glider_SkullTrooper_FlyLoop_01}}
 {id=WavePlayer_10, properties {soundWaveAssetPtr=&/Glider_SkullTrooper_Vocal_03}}
 {id=WavePlayer_11, properties {soundWaveAssetPtr=&/Glider_SkullTrooper_Vocal_04}}
 {id=WavePlayer_8, properties {soundWaveAssetPtr=&/Glider_SkullTrooper_Vocal_01}}
 {id=WavePlayer_9, properties {soundWaveAssetPtr=&/Glider_SkullTrooper_Vocal_02}}
]
properties {
 attenuationSettings=&Fort_Player_Glider_Thrust_Athena_Att
 audioPropertiesSheet=&APS_Cosmetic_Glider_Loop
 maxDistance=12500
 soundClassObject=&Fort_Gliders_Land
 volumeMultiplier=0.45
}
