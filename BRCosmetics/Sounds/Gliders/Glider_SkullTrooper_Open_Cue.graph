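# Sound cue for the Skull Trooper glider's opening sound, layering a deploy effect with a delayed vocal.
# A mixer blends the main open wave with a quieter branch.
# That branch waits a fixed delay, then plays one of three equally weighted vocal takes at a slightly raised, randomized pitch.
# Glider open/close attenuation, a cosmetic glider-open property sheet and a shared concurrency set govern playback.
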SoundCue /BRCosmetics/Sounds/Gliders/Glider_SkullTrooper_Open_Cue
root=Mixer_0
nodes [
 {id=Delay_0, children [Random_0], properties {delayMax=1.6, delayMin=1.6}}
 {id=Mixer_0, children [WavePlayer_0,Modulator_0], properties {inputVolume [0.5,0.2]}}
 {id=Modulator_0, children [Delay_0], properties {pitchMax=1.2, pitchMin=1.1}}
 {id=Random_0, children [WavePlayer_4,WavePlayer_5,WavePlayer_6], properties {weights [1,1,1]}}
 {id=WavePlayer_0, properties {soundWaveAssetPtr=&/Glider_SkullTrooper_Open_01}}
 {id=WavePlayer_4, properties {soundWaveAssetPtr=&/Glider_SkullTrooper_OpenVocal_01}}
 {id=WavePlayer_5, properties {soundWaveAssetPtr=&/Glider_SkullTrooper_OpenVocal_02}}
 {id=WavePlayer_6, properties {soundWaveAssetPtr=&/Glider_SkullTrooper_OpenVocal_03}}
]
properties {
 attenuationSettings=/Game/Sounds/Attenuation/Fort_Player_Glider_OpenClose_Athena_Att
 audioPropertiesSheet=/BRCosmetics/Sounds/Settings/PropertySheets/Glider/APS_Cosmetic_Glider_Open
 maxDistance=10000
 soundClassObject=&Fort_Gliders_Land
 volumeMultiplier=1
 concurrencySet [&Athena_GliderOpenClose_Conc]
}
