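# Sound cue for the closing of the Spider glider cosmetic, routed through the glider landing sound class.
# A Mixer layers a close sound, pitched down randomly by a Modulator, with an unmodulated vocal close wave at a lower input volume.
# Glider open/close attenuation and concurrency, plus a cosmetic glider close property sheet, govern playback, with overall pitch and volume reduced on the cue.
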SoundCue /BRCosmetics/Sounds/Gliders/Glider_Spider_Close_Cue
root=Mixer_1
nodes [
 {id=Mixer_1, children [Modulator_0,WavePlayer_3], properties {inputVolume [0.8,0.6]}}
 {id=Modulator_0, children [WavePlayer_0], properties {pitchMax=0.9, pitchMin=0.8}}
 {id=WavePlayer_0, properties {soundWaveAssetPtr=&/Glider_Spider_Close_01}}
 {id=WavePlayer_3, properties {soundWaveAssetPtr=&/Glider_Spider_CloseVocal_01}}
]
properties {
 attenuationSettings=/Game/Sounds/Attenuation/Fort_Player_Glider_OpenClose_Athena_Att
 audioPropertiesSheet=/BRCosmetics/Sounds/Settings/PropertySheets/Glider/APS_Cosmetic_Glider_Close
 maxDistance=10000
 pitchMultiplier=0.9
 soundClassObject=&Fort_Gliders_Land
 volumeMultiplier=0.6
 concurrencySet [&Athena_GliderOpenClose_Conc]
}
