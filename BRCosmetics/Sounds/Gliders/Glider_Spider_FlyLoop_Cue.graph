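# Sound cue for the spider glider's flight loop, layering a continuous fly sound with intermittent creature vocals.
# A mixer blends the looping fly wave, pitched slightly down, with a looped branch that waits a randomized delay of several seconds, then plays one of four equally weighted vocals through a modulator.
# It uses the shared player glider thrust attenuation and cosmetic glider loop audio property sheet.
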SoundCue /BRCosmetics/Sounds/Gliders/Glider_Spider_FlyLoop_Cue
root=Mixer_1
nodes [
 {id=Delay_1, children [Modulator_2], properties {delayMax=9, delayMin=5}}
 {id=Looping_1, children [Delay_1]}
 {id=Mixer_1, children [Modulator_3,Looping_1], properties {inputVolume [0.7,0.75]}}
 {id=Modulator_2, children [Random_2]}
 {id=Modulator_3, children [WavePlayer_8], properties {pitchMax=0.9, pitchMin=0.9}}
 {id=Random_2, children [WavePlayer_9, WavePlayer_10, WavePlayer_11, WavePlayer_12], properties {weights [1,1,1,1]}}
 {id=WavePlayer_10, properties {soundWaveAssetPtr=&/Glider_Spider_Vocal_02}}
 {id=WavePlayer_11, properties {soundWaveAssetPtr=&/Glider_Spider_Vocal_03}}
 {id=WavePlayer_12, properties {soundWaveAssetPtr=&/Glider_Spider_Vocal_04}}
 {id=WavePlayer_8, properties {bLooping=true, soundWaveAssetPtr=&/Glider_Spider_FlyLoop_01}}
 {id=WavePlayer_9, properties {soundWaveAssetPtr=&/Glider_Spider_Vocal_01}}
]
properties {
 attenuationSettings=&Fort_Player_Glider_Thrust_Athena_Att
 audioPropertiesSheet=&APS_Cosmetic_Glider_Loop
 maxDistance=12500
 soundClassObject=&Fort_Gliders_Land
 volumeMultiplier=0.35
}
