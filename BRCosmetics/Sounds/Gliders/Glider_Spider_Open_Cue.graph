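# Sound cue for the opening of the Spider glider cosmetic, layering a mechanical open sound with a vocal.
# A mixer plays the open wave immediately alongside a quieter vocal branch, which is held back by a fixed delay and randomly pitched down slightly.
# It uses the shared player glider open/close attenuation and concurrency, the cosmetic glider-open property sheet and a glider sound class.
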SoundCue /BRCosmetics/Sounds/Gliders/Glider_Spider_Open_Cue
root=Mixer_1
nodes [
 {id=Delay_1, children [WavePlayer_5], properties {delayMax=1.6, delayMin=1.6}}
 {id=Mixer_1, children [WavePlayer_4,Modulator_1], properties {inputVolume [0.9,0.6]}}
 {id=Modulator_1, children [Delay_1], properties {pitchMax=0.95, pitchMin=0.85}}
 {id=WavePlayer_4, properties {soundWaveAssetPtr=&/Glider_Spider_Open_01}}
 {id=WavePlayer_5, properties {soundWaveAssetPtr=&/Glider_Spider_OpenVocal_01}}
]
properties {
 attenuationSettings=/Game/Sounds/Attenuation/Fort_Player_Glider_OpenClose_Athena_Att
 audioPropertiesSheet=/BRCosmetics/Sounds/Settings/PropertySheets/Glider/APS_Cosmetic_Glider_Open
 maxDistance=10000
 soundClassObject=&Fort_Gliders_Land
 volumeMultiplier=0.9
 concurrencySet [&Athena_GliderOpenClose_Conc]
}
