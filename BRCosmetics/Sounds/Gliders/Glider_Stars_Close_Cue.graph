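# Sound cue that plays the closing sound for the Stars cosmetic glider.
# A single wave player feeds the Glider_Stars_Close wave through a root Modulator left at default settings.
# Spatialization and falloff come from the shared Athena glider open/close attenuation asset.
# It also uses the cosmetic glider-close audio property sheet and routes to the glider landing sound class.
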SoundCue /BRCosmetics/Sounds/Gliders/Glider_Stars_Close_Cue
root=Modulator_0
nodes [{id=Modulator_0, children [WavePlayer_0]}, {id=WavePlayer_0, properties {soundWaveAssetPtr=&/Glider_Stars_Close}}]
properties {
 attenuationSettings=/Game/Sounds/Attenuation/Fort_Player_Glider_OpenClose_Athena_Att
 audioPropertiesSheet=/BRCosmetics/Sounds/Settings/PropertySheets/Glider/APS_Cosmetic_Glider_Close
 maxDistance=10000
 soundClassObject=&Fort_Gliders_Land
 volumeMultiplier=1
}
