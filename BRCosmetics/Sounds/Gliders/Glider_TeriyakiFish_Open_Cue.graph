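# Sound cue played when the TeriyakiFish glider cosmetic opens.
# A mixer layers the open sound at full volume with a quieter cheer branch.
# The cheer branch waits a fixed short delay, then randomly picks one of three equally weighted cheer waves.
# It uses the shared player glider open/close attenuation and concurrency, a glider-open audio property sheet, and the glider land sound class.
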SoundCue /BRCosmetics/Sounds/Gliders/Glider_TeriyakiFish_Open_Cue
root=Mixer_0
nodes [
 {id=Delay_0, children [Random_0], properties {delayMax=0.6, delayMin=0.6}}
 {id=Mixer_0, children [WavePlayer_1,Delay_0], properties {inputVolume [1,0.25]}}
 {id=Random_0, children [WavePlayer_0,WavePlayer_2,WavePlayer_3], properties {weights [1,1,1]}}
 {id=WavePlayer_0, properties {soundWaveAssetPtr=&/Glider_TeriyakiFish_Cheers_01}}
 {id=WavePlayer_1, properties {soundWaveAssetPtr=&/Glider_TeriyakiFish_Open_01}}
 {id=WavePlayer_2, properties {soundWaveAssetPtr=&/Glider_TeriyakiFish_Cheers_02}}
 {id=WavePlayer_3, properties {soundWaveAssetPtr=&/Glider_TeriyakiFish_Cheers_03}}
]
properties {
 attenuationSettings=/Game/Sounds/Attenuation/Fort_Player_Glider_OpenClose_Athena_Att
 audioPropertiesSheet=/BRCosmetics/Sounds/Settings/PropertySheets/Glider/APS_Cosmetic_Glider_Open
 maxDistance=10000
 soundClassObject=&Fort_Gliders_Land
 volumeMultiplier=1.2
 concurrencySet [&Athena_GliderOpenClose_Conc]
}
