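# Sound cue for the rowing sound of the Viking glider cosmetic.
# A quality-level switch sends its first two slots to an equally weighted random pick of three row waves.
# Its last two slots instead play a single row wave through a modulator at default settings.
# Output is at reduced volume, using the player glider thrust attenuation and the glider landing sound class.
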SoundCue /BRCosmetics/Sounds/Gliders/Glider_Viking_Row_Cue
root=QualityLevel_0
nodes [
 {id=Modulator_0, children [WavePlayer_0]}
 {id=QualityLevel_0, children [Random_0,Random_0,Modulator_0,Modulator_0]}
 {id=Random_0, children [WavePlayer_1,WavePlayer_2,WavePlayer_3], properties {weights [1,1,1]}}
 {id=WavePlayer_0, properties {soundWaveAssetPtr=&/Glider_Viking_Row_01}}
 {id=WavePlayer_1, properties {soundWaveAssetPtr=&Glider_Viking_Row_01}}
 {id=WavePlayer_2, properties {soundWaveAssetPtr=&/Glider_Viking_Row_02}}
 {id=WavePlayer_3, properties {soundWaveAssetPtr=&/Glider_Viking_Row_03}}
]
properties {
 attenuationSettings=&Fort_Player_Glider_Thrust_Athena_Att
 maxDistance=12500
 soundClassObject=&Fort_Gliders_Land
 volumeMultiplier=0.5
}
