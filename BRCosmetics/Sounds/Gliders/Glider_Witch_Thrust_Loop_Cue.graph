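# Sound cue for the looping thrust sound of the Witch glider cosmetic.
# A looping wave player feeds an Enveloper whose volume curve ramps up from a low level over a brief onset.
# The cue doubles pitch, lowers overall volume, and uses player glider thrust attenuation, a cosmetic glider loop audio properties sheet and a glider sound class.
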SoundCue /BRCosmetics/Sounds/Gliders/Glider_Witch_Thrust_Loop_Cue
root=Enveloper_0
nodes [
 {
  id=Enveloper_0
  children [WavePlayer_1]
  properties {pitchCurve {editorCurveData {keys [{time=0}]}}, volumeCurve {editorCurveData {keys [<at 0 0.2>,<at 0.3 1>]}}}
 }
 {id=WavePlayer_1, properties {bLooping=true, soundWaveAssetPtr=&/Glider_Witch_Thrust_01}}
]
properties {
 attenuationSettings=&Fort_Player_Glider_Thrust_Athena_Att
 audioPropertiesSheet=&APS_Cosmetic_Glider_Loop
 maxDistance=12500
 pitchMultiplier=2
 soundClassObject=&Fort_Gliders_Land
 volumeMultiplier=0.3
}
