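# Sound cue for the wing flap of the Witch glider cosmetic.
# It picks one of two flap waves at random with equal weight, then a modulator varies pitch slightly around normal.
# It plays quietly through player glider thrust attenuation and a glider sound class, with a concurrency override capping simultaneous instances.
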SoundCue /BRCosmetics/Sounds/Gliders/Glider_Witch_WingFlap_Cue
root=Modulator_0
nodes [
 {id=Modulator_0, children [Random_2], properties {pitchMax=1.1, pitchMin=0.9}}
 {id=Random_2, children [WavePlayer_7,WavePlayer_8], properties {weights [1,1]}}
 {id=WavePlayer_7, properties {soundWaveAssetPtr=&/Glider_Witch_Flap_01}}
 {id=WavePlayer_8, properties {soundWaveAssetPtr=&/Glider_Witch_Flap_02}}
]
properties {
 attenuationSettings=&Fort_Player_Glider_Thrust_Athena_Att
 bOverrideConcurrency=true
 maxDistance=12500
 soundClassObject=&Fort_Gliders_Land
 volumeMultiplier=0.4
 concurrencyOverrides {maxCount=5}
}
