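# Sound cue for the Yeti glider's opening in a battle royale cosmetics set.
# A mixer layers the glider open wave at full volume with a quieter yeti vocalization, which a random node picks from three equally weighted variants.
# Player glider open/close attenuation, a glider-open property sheet, a glider sound class and a shared open/close concurrency set govern its playback.
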SoundCue /BRCosmetics/Sounds/Gliders/Glider_Yeti_Open_Cue
root=Mixer_0
nodes [
 {id=Mixer_0, children [WavePlayer_1,Random_0], properties {inputVolume [1,0.7]}}
 {id=Random_0, children [WavePlayer_2,WavePlayer_3,WavePlayer_4], properties {weights [1,1,1]}}
 {id=WavePlayer_1, properties {soundWaveAssetPtr=&/Glider_Yeti_Open}}
 {id=WavePlayer_2, properties {soundWaveAssetPtr=&/Glider_Yeti_Vox_Open_01}}
 {id=WavePlayer_3, properties {soundWaveAssetPtr=&/Glider_Yeti_Vox_Open_02}}
 {id=WavePlayer_4, properties {soundWaveAssetPtr=&/Glider_Yeti_Vox_Open_03}}
]
properties {
 attenuationSettings=/Game/Sounds/Attenuation/Fort_Player_Glider_OpenClose_Athena_Att
 audioPropertiesSheet=/BRCosmetics/Sounds/Settings/PropertySheets/Glider/APS_Cosmetic_Glider_Open
 maxDistance=10000
 soundClassObject=&Fort_Gliders_Land
 volumeMultiplier=1
 concurrencySet [&Athena_GliderOpenClose_Conc]
}
