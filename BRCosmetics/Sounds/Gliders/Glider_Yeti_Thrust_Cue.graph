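# Sound cue for the Yeti glider's thrust, layering a thrust sound with an occasional Yeti vocalization.
# A mixer plays one of three random thrust waves immediately, alongside a branch delayed by 0.2 seconds.
# The delayed branch pairs a random pick of three vox waves with an equally weighted empty slot, so the voice plays about half the time.
# Glider-specific attenuation, concurrency and sound class govern playback.
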SoundCue /BRCosmetics/Sounds/Gliders/Glider_Yeti_Thrust_Cue
root=Mixer_0
nodes [
 {id=Delay_0, children [Random_2], properties {delayMax=0.2, delayMin=0.2}}
 {id=Mixer_0, children [Random_0,Delay_0], properties {inputVolume [1,1]}}
 {id=Random_0, children [WavePlayer_1,WavePlayer_2,WavePlayer_3], properties {weights [1,1,1]}}
 {id=Random_1, children [WavePlayer_4,WavePlayer_5,WavePlayer_6], properties {weights [1,1,1]}}
 {id=Random_2, children [Random_1,null], properties {weights [1,1]}}
 {id=WavePlayer_1, properties {soundWaveAssetPtr=&/Glider_Yeti_Thrust_01}}
 {id=WavePlayer_2, properties {soundWaveAssetPtr=&/Glider_Yeti_Thrust_02}}
 {id=WavePlayer_3, properties {soundWaveAssetPtr=&/Glider_Yeti_Thrust_03}}
 {id=WavePlayer_4, properties {soundWaveAssetPtr=&/Glider_Yeti_Vox_Thrust_01}}
 {id=WavePlayer_5, properties {soundWaveAssetPtr=&/Glider_Yeti_Vox_Thrust_02}}
 {id=WavePlayer_6, properties {soundWaveAssetPtr=&/Glider_Yeti_Vox_Thrust_03}}
]
properties {
 attenuationSettings=&Fort_Player_Glider_Thrust_Athena_Att
 audioPropertiesSheet=&APS_Cosmetic_Glider_Loop
 maxDistance=12500
 soundClassObject=&Fort_Gliders_Land
 volumeMultiplier=0.5
 concurrencySet [&Athena_Glider_Thrust_Conc]
}
