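# Sound cue for the close sound of the Durrrburger cosmetic glider in Battle Royale.
# A single wave player feeds the Durrburger close wave straight to the root, with no randomization or modulation.
# It uses the shared glider open/close attenuation and concurrency, the cosmetic glider-close property sheet, and the glider landing sound class.
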SoundCue /BRCosmetics/Sounds/Gliders/SFX_Glider_Durrrburger_Close_Cue
root=WavePlayer_2
nodes [{id=WavePlayer_2, properties {soundWaveAssetPtr=&/SFX_Glider_Close_Durrburger}}]
properties {
 attenuationSettings=/Game/Sounds/Attenuation/Fort_Player_Glider_OpenClose_Athena_Att
 audioPropertiesSheet=/BRCosmetics/Sounds/Settings/PropertySheets/Glider/APS_Cosmetic_Glider_Close
 maxDistance=10000
 soundClassObject=&Fort_Gliders_Land
 concurrencySet [&Athena_GliderOpenClose_Conc]
}
