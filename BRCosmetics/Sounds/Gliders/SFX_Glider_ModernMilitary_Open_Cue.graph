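# Sound cue for the Modern Military glider cosmetic's open sound.
# A mixer layers the military glider open wave with a quieter wind-start wave, which is delayed by a fixed interval and pitched down through a modulator.
# Shared glider open/close attenuation and concurrency assets, a cosmetic glider-open property sheet and a glider sound class govern playback.
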SoundCue /BRCosmetics/Sounds/Gliders/SFX_Glider_ModernMilitary_Open_Cue
root=Mixer_0
nodes [
 {id=Delay_0, children [Modulator_0], properties {delayMax=1, delayMin=1}}
 {id=Mixer_0, children [WavePlayer_0,Delay_0], properties {inputVolume [1,0.6]}}
 {id=Modulator_0, children [WavePlayer_1], properties {pitchMax=0.8, pitchMin=0.8}}
 {id=WavePlayer_0, properties {soundWaveAssetPtr=&/SFX_Glider_Military_Open_01}}
 {id=WavePlayer_1, properties {soundWaveAssetPtr=&/SFX_Glider_Military_WindStart_01}}
]
properties {
 attenuationSettings=/Game/Sounds/Attenuation/Fort_Player_Glider_OpenClose_Athena_Att
 audioPropertiesSheet=/BRCosmetics/Sounds/Settings/PropertySheets/Glider/APS_Cosmetic_Glider_Open
 maxDistance=10000
 soundClassObject=&Fort_Gliders_Land
 concurrencySet [&Athena_GliderOpenClose_Conc]
}
